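# Sound cue for the Ball event sequence audio in Athena, routed to the scripted 3D events sound class.
# A single wave player feeds Ball_Sequence_01 straight to the output at boosted volume and raised priority.
# Attenuation overrides turn off distance falloff, so listener focus lowers the volume when the source is off-axis.
# A very wide stereo spread and large non-spatialized radius suit a huge, map-scale source.
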SoundCue /Game/Athena/Sounds/Events/Ball/Ball_Sequence_Cue
root=WavePlayer_0
nodes [{id=WavePlayer_0, properties {soundWaveAssetPtr=&/Ball_Sequence_01}}]
properties {
 bOverrideAttenuation=true
 maxDistance=2097152
 priority=10
 soundClassObject=/Game/Sounds/FortSoundClassesAndModes/Fort_Events_Scripted_3D
 volumeMultiplier=2
 attenuationOverrides {
  bAttenuate=false
  bEnableFocusInterpolation=true
  bEnableListenerFocus=true
  focusAzimuth=20
  focusReleaseInterpSpeed=0.7
  nonFocusAzimuth=40
  nonFocusVolumeAttenuation=0.6
  nonSpatializedRadiusStart=10000
  stereoSpread=100000
 }
}
